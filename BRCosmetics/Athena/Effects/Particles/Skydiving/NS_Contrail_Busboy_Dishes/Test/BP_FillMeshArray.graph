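# Editor utility that copies selected static mesh actors into Niagara array parameters on selected Niagara actors, in a contrail effect test folder.
# NiagaraFillMeshArray collects each mesh with world scale and rotation, its unscaled local bounds size, and its position.
# Positions are world locations or offsets from the selection's bounds center, optionally using the bottom as Z.
# It then writes the three arrays and mesh count to each Niagara component by configurable names.
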EditorUtilityBlueprint /BRCosmetics/Athena/Effects/Particles/Skydiving/NS_Contrail_Busboy_Dishes/Test/BP_FillMeshArray
graphs [
 EventGraph {links [], nodes []}
 NiagaraFillMeshArray {nodes [
   n1 = NiagaraFillMeshArray!() -> (
    then
    Mesh Array Name:Name=Mesh Array
    Offset Array Name:Name=Mesh Offset Array
    Dimensions Array Name:Name=Mesh Dimensions Array
    Mesh Count Name:Name=Mesh Count
    Offsets From Center:bool=true
    Use Base as Center:bool
   ) {is=title K2Node_FunctionEntry}
   > Set Mesh Array Name(Mesh Array Name:Name=@n1) -> (Output_Get:Name=None)
   > Set Offset Array Name(Offset Array Name:Name=@n1.Offset Array Name) -> (Output_Get:Name=None)
   > Set Dimensions Array Name(Dimensions Array Name:Name=@n1.Dimensions Array Name) -> (Output_Get:Name=None)
   > Set Mesh Count Name(Mesh Count Name:Name=@n1.Mesh Count Name) -> (Output_Get:Name=None)
   > Set Offsets from Center(Offsets from Center:bool=@n1.Offsets From Center) -> (Output_Get:bool)
   > Set Use Base as Center(Use Base as Center:bool=@n1.Use Base as Center) -> (Output_Get:bool)
   > n2 = GetSelectedActors()
   > Set Selected Actors(Selected Actors:Array of Actor Object References=@n2) -> (Output_Get:Array of Actor Object References)
   > n3 = Sequence!(execute) -> (then_0, then_1, then_2:Exec)
   > n4 = Filter by Class!(
    execute
    TargetArray:Array of Object References=Get Selected Actors() -> (Selected Actors:Array of Actor Object References)
    ObjectClass:Object Class Reference=&StaticMeshActor
    FilterType:EEditorScriptingFilterType Enum
   ) -> (then, ReturnValue:Array of Static Mesh Actor Object References)
   > Set Mesh Count(Mesh Count:int=Length(TargetArray:Array of Static Mesh Actor Object References=@n4)) -> (Output_Get:int=0)
   > n5 = GetActorArrayBounds!(execute, Actors:Array of Actor Object References=@n4, bOnlyCollidingComponents:bool=false) -> (then, Center:Vector, BoxExtent:Vector)
   > Set Center(Center:Vector=@n5) -> (Output_Get:Vector)
   n6 = Break Box!(Box:"Box Structure (by ref)"=MakeBoxFromCenterExtents!(Center:Vector=@n5, Extents:Vector=@n5.BoxExtent) -> (ReturnValue:Box Structure)) -> (Min:Vector=~, Max:Vector=~, IsValid:bool=~) {is=title K2Node_BreakStruct}
   > Set Min Bounds(Min Bounds:Vector=@n6) -> (Output_Get:Vector)
   > Set Max Bounds(Max Bounds:Vector=@n6.Max) -> (Output_Get:Vector)
   n7 = BreakVector(InVec=@n5)
   > Set Base(Base:Vector=MakeVector(X=@n7, Y=@n7.Y, Z=BreakVector(InVec=@n6).Z)) -> (Output_Get:Vector)
   > n8 = For Each Loop(Array:Array of Static Mesh Actor Object References=@n4) -> (Array Element:Static Mesh Actor Object Reference)
   n9 = Get StaticMeshComponent(self=@n8)
   > Add!(
    execute
    TargetArray:Array of Niagara Mesh Renderer Mesh Properties Base Structures=Get Niagara Mesh Properties() -> (Niagara Mesh Properties:Array of Niagara Mesh Renderer Mesh Properties Base Structures)
    NewItem_Mesh:Static Mesh Object Reference=Get StaticMesh(self=@n9)
    NewItem_Scale:Vector=Get World Scale!(self:Scene Component Object Reference=@n9) -> (ReturnValue:Vector) {is=title K2Node_CallFunction}
    NewItem_Rotation:Rotator=Get World Rotation!(self:Scene Component Object Reference=@n9) -> (ReturnValue:Rotator) {is=title K2Node_CallFunction}
    NewItem_PivotOffset:Vector="0,0,0"
    NewItem_PivotOffsetSpace:ENiagaraMeshPivotOffsetSpace Enum=Mesh
   ) -> (then, ReturnValue) {is=title K2Node_CallArrayFunction}
   > n10 = Branch(Condition=Get Offsets from Center() -> (Offsets from Center:bool))
   n11 = Get Center() -> (Center:Vector)
   n12 = Get Use Base as Center() -> (Use Base as Center:bool)
   n13 = Get Base() -> (Base:Vector)
   n14 = Get World Location!(self:Scene Component Object Reference=@n9) -> (ReturnValue:Vector) {is=title K2Node_CallFunction}
   > n15 = Add!(
    execute
    TargetArray:Array of Vectors=Get Offset Array() -> (Offset Array:Array of Vectors)
    NewItem:"Vector (by ref)"=vector - vector(A=@n14, B=SelectVector(A=@n13, B=@n11, bPickA=@n12))
   ) -> (then, ReturnValue)
   n16 = GetLocalBounds!(self:Static Mesh Component Object Reference=@n9) -> (Min:Vector, Max:Vector)
   Add(
    execute=[@n17.then, @n15.then]
    TargetArray=Get Dimensions Array() -> (Dimensions Array:Array of Vectors)
    NewItem=vector - vector(A=@n16.Max, B=@n16)
   )
   n17 = Add(execute=@n10.else, TargetArray=Get Offset Array(), NewItem=@n14)
   n18 = Filter by Class(execute=@n3.then_1, TargetArray=Get Selected Actors(), ObjectClass=/Script/Niagara.NiagaraActor) -> (ReturnValue:Array of Niagara Particle System Actor Object References)
   > n19 = For Each Loop(Array:Array of Niagara Particle System Actor Object References=@n18) -> (Array Element:Niagara Particle System Actor Object Reference)
   n20 = Get NiagaraComponent!(self:Niagara Particle System Actor Object Reference=@n19) -> (NiagaraComponent:Niagara Particle System Component Object Reference)
   > Niagara Set Mesh Array!(
    execute
    NiagaraComponent:Niagara Particle System Component Object Reference=@n20
    OverrideName:Name=Get Mesh Array Name() -> (Mesh Array Name:Name=None)
    ArrayData:Array of Niagara Mesh Renderer Mesh Properties Base Structures=Get Niagara Mesh Properties()
   ) -> (then)
   > Niagara Set Vector Array!(
    execute
    NiagaraSystem:Niagara Particle System Component Object Reference=@n20
    OverrideName:Name=Get Offset Array Name() -> (Offset Array Name:Name=None)
    ArrayData:Array of Vectors=Get Offset Array()
   ) -> (then)
   > Niagara Set Vector Array(NiagaraSystem=@n20, OverrideName=Get Dimensions Array Name() -> (Dimensions Array Name:Name=None), ArrayData=Get Dimensions Array())
   > "Set Niagara Variable (Int32)"!(
    execute
    self:Niagara Particle System Component Object Reference=@n20
    InVariableName:Name=Get Mesh Count Name() -> (Mesh Count Name:Name=None)
    InValue:int=Get Mesh Count() -> (Mesh Count:int=0)
   ) -> (then) {is=title K2Node_CallFunction}
  ]}
]
variables [
 Base {type=Vector}
 Center {type=Vector}
 Dimensions Array {type="array<Vector>"}
 Dimensions Array Name {type=string, default=None}
 Max Bounds {type=Vector, default=<-9999 -9999 -9999>}
 Mesh Array Name {type=string, default=Mesh Array}
 Mesh Count {type=integer}
 Mesh Count Name {type=string, default=Mesh Count}
 Min Bounds {type=Vector, default=<9999 9999 9999>}
 Niagara Mesh Properties {type="array<NiagaraMeshRendererMeshPropertiesBase>", default []}
 Offset Array {type="array<Vector>"}
 Offset Array Name {type=string, default=Offset Array}
 Offsets from Center {type=boolean, default=true}
 Selected Actors {type="array<Actor>"}
 Use Base as Center {type=boolean}
]
